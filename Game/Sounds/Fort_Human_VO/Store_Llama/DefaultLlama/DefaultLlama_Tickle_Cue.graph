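# Sound cue for the default Store Llama's tickle voice reaction.
# A Random node picks among eight equally weighted Dialogue Players, each playing one tickle Dialogue Wave with the DefaultLlama dialogue voice as speaker and no targets.
# The cue routes through the narrator VO sound class and references the StW llama VO concurrency set to limit overlapping lines.
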SoundCue /Game/Sounds/Fort_Human_VO/Store_Llama/DefaultLlama/DefaultLlama_Tickle_Cue
root=Random_2
nodes [
 {
  id=DialoguePlayer_0
  properties {dialogueWaveParameter {
    dialogueWave=&/JB_BasicLlama_Tickle_01_Dialogue
    context {speaker=&/DefaultLlama_DV, targets []}
   }}
 }
 {
  id=DialoguePlayer_1
  properties {dialogueWaveParameter {
    dialogueWave=&/JB_BasicLlama_Tickle_02_Dialogue
    context {speaker=&DefaultLlama_DV, targets []}
   }}
 }
 {
  id=DialoguePlayer_2
  properties {dialogueWaveParameter {
    dialogueWave=&/JB_BasicLlama_Tickle_03_Dialogue
    context {speaker=&DefaultLlama_DV, targets []}
   }}
 }
 {
  id=DialoguePlayer_3
  properties {dialogueWaveParameter {
    dialogueWave=&/JB_BasicLlama_Tickle_04_Dialogue
    context {speaker=&DefaultLlama_DV, targets []}
   }}
 }
 {
  id=DialoguePlayer_4
  properties {dialogueWaveParameter {
    dialogueWave=&/JB_BasicLlama_Tickle_05_Dialogue
    context {speaker=&DefaultLlama_DV, targets []}
   }}
 }
 {
  id=DialoguePlayer_5
  properties {dialogueWaveParameter {
    dialogueWave=&/JB_BasicLlama_Tickle_06_Dialogue
    context {speaker=&DefaultLlama_DV, targets []}
   }}
 }
 {
  id=DialoguePlayer_6
  properties {dialogueWaveParameter {
    dialogueWave=&/JB_BasicLlama_Tickle_07_Dialogue
    context {speaker=&DefaultLlama_DV, targets []}
   }}
 }
 {
  id=DialoguePlayer_7
  properties {dialogueWaveParameter {
    dialogueWave=&/JB_BasicLlama_Tickle_08_Dialogue
    context {speaker=&DefaultLlama_DV, targets []}
   }}
 }
 {
  id=Random_2
  children [
   DialoguePlayer_0
   DialoguePlayer_1
   DialoguePlayer_2
   DialoguePlayer_3
   DialoguePlayer_4
   DialoguePlayer_5
   DialoguePlayer_6
   DialoguePlayer_7
  ]
  properties {weights [1,1,1,1,1,1,1,1]}
 }
]
properties {
 maxDistance=2097152
 priority=0.9
 soundClassObject=/Game/Sounds/FortSoundClassesAndModes/Fort_VO_Narrator
 volumeMultiplier=1
 concurrencyOverrides {maxCount=1, resolutionRule=StopLowestPriority}
 concurrencySet [/Game/Sounds/Concurrency_Rules/StW_Llama_VO_Conc]
}
